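# M_2DToon_Debris is an unlit material for Niagara mesh particles that gives melee weapon toon-effect debris a two-tone, cel-shaded look.
# A fixed tilted tangent-space vector is transformed to world space and dotted with world up to fake top-down lighting.
# A hard threshold on that result picks between two dark grey levels, which feed the emissive input of MF_FX_EmissiveAdjustments.
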Material /Game/Weapons/FORT_Melee/Effects/ToonShader/M_2DToon_Debris
graph {
 properties {
  bUsedWithNiagaraMeshParticles=true
  pixelDepthOffsetMode=PDOM_Legacy
  previewMesh=/Game/Effects/Fort_Effects/Mesh/Ice/SM_Contrail_IceShard
  shadingModel=MSM_Unlit
  usedShadingModels=Unlit
 }
 nodes [
  n1 = Constant3Vector() {constant=<rgba 0.2 -0.4 1 0>}
  n2 = Saturate(None=DotProduct(A=Transform(None=@n1), B=Constant3Vector() {constant=<rgba 0 0 1 0>}))
  n3 = Constant() {r=0.07}
  n4 = Constant() {r=1}
  n5 = Constant()
  MaterialFunctionCall(Emissive=LinearInterpolate(Alpha=If(A=@n2, B=@n3, "A > B"=@n4, "A == B"=@n4, "A < B"=@n5)) {constA=0.1, constB=0.2}) {materialFunction=&MF_FX_EmissiveAdjustments}
 ]
}
