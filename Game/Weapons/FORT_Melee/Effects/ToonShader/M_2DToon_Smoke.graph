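# Unlit translucent Niagara sprite and mesh-particle material that draws toon smoke from a 3x3 flipbook texture.
# Dynamic parameter Time drives the frame, and Picker chooses alpha or blue for optional flat opacity, which otherwise uses red.
# Particle color, optionally multiplied by red, is dimmed and passed through emissive adjustments.
# Opacity scales with particle alpha, with depth fade skipped on forward and mobile paths.
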Material /Game/Weapons/FORT_Melee/Effects/ToonShader/M_2DToon_Smoke
graph {
 properties {
  bUsedWithNiagaraMeshParticles=true
  bUsedWithNiagaraSprites=true
  blendMode=BLEND_Translucent
  pixelDepthOffsetMode=PDOM_Legacy
  shadingModel=MSM_Unlit
  usedShadingModels=Unlit
 }
 nodes [
  n1 = ParticleColor()
  n2 = DynamicParameter() {paramNames [Time,Picker,Param3,Param4]}
  n3 = MaterialFunctionCall(
   "Animation  Phase (0-1)"=@n2.Time
   Number of Rows=Constant() {r=3}
   Number of Columns=Constant() {r=3}
   Texture=TextureObjectParameter() {parameterName=Smoke Texture, texture=&/T_2DToon_Flipbook_Smoke}
  ) {materialFunction=&FlipBook}
  n4 = MaterialFunctionCall(Float3=@n3.Result) {materialFunction=&BreakOutFloat3Components}
  n5 = LinearInterpolate(A=@n4.R, B=@n4.R, Alpha=@n2.Picker)
  n6 = Multiply(A=@n1.RGB, B=@n5)
  n7 = ScalarParameter() {parameterName=MultiplyParticleColorWithTexture}
  MaterialFunctionCall(Emissive=Multiply(A=LinearInterpolate(A=@n1.RGB, B=@n6, Alpha=@n7), B=Constant() {r=0.3})) {materialFunction=&MF_FX_EmissiveAdjustments}
  MaterialFunctionCall!(
   "Animation  Phase (0-1)"
   Number of Columns X
   Number of Rows Y
   Texture
   UVs
   Motion Vector Texture
   Phase to Debug
   Motion Vector Intensity
   Double Apply Motion Vectors
   Custom UV Input
   Custom UV2 Input
   Use Custom UVs
  ) {materialFunction=/Engine/Functions/Engine_MaterialFunctions02/Texturing/FlipBook_MotionVectors}
  n8 = MaterialFunctionCall(Float3=@n3.Result) {materialFunction=&BreakOutFloat3Components}
  n9 = Multiply(
   A=@n1.A
   B=LinearInterpolate(
    A=@n5
    B=LinearInterpolate(A=@n3.Alpha, B=@n8.B, Alpha=@n2.Picker)
    Alpha=ScalarParameter() {parameterName="UseFlatAlpha (B && A Channels in Texture)"}
   )
  )
  n10 = DepthFade(Opacity=@n9, FadeDistance=ScalarParameter() {defaultValue=12, parameterName=DepthFadeDis})
  ShadingPathSwitch(Default=@n10, Deferred=@n10, Forward=@n9, Mobile=@n9)
 ]
}
